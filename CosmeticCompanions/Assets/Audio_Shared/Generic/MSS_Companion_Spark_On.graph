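# One-shot mono MetaSound Source for a cosmetic companion's "Spark On" sound, in the pets sound class with spark attenuation.
# On play, a trigger chain picks a wave from an array (one entry by default), rolls random pitch and volume, then fires the wave player.
# Pitch shifts the player and volume multiplies its output; playback end signals OnFinished.
# Concurrency allows one instance per owner and blocks new ones.
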MetaSoundSource /CosmeticCompanions/Assets/Audio_Shared/Generic/MSS_Companion_Spark_On
inputs [
 Random Pitch Max {type=Float}
 Random Pitch Min {type=Float}
 Random Volume Change Max {type=Float}
 Random Volume Change Min {type=Float}
 UE.Source.OnPlay {type=Trigger}
 Wave Array {type=WaveAsset:Array}
]
interfaces [UE.OutputFormat.Mono:1.0, UE.Source.OneShot:1.0, UE.Source:1.1]
outputs [
 On Play {type=Trigger}
 "UE.OutputFormat.Mono.Audio:0" {type=Audio}
 UE.Source.OneShot.OnFinished {type=Trigger}
]
pages [Default {
  input_defaults {
   Random Pitch Max=0.5
   Random Pitch Min=-0.5
   Random Volume Change Max=1.2
   Random Volume Change Min=0.8
   UE.Source.OnPlay=null
   Wave Array [&/Companion_Spark_On_01]
  }
  nodes [
   ArrayRandomGet_0 = Array.Random Get:WaveAsset:Array(Next=^UE.Source.OnPlay, In Array=^Wave Array) {also {defaults {Weights []}}}
   n1 = UE.RandomFloat(Max=^Random Pitch Max, Min=^Random Pitch Min, Next=@ArrayRandomGet_0.On Next)
   n2 = UE.RandomFloat(Max=^Random Volume Change Max, Min=^Random Volume Change Min, Next=@n1.On Next)
   WavePlayer_0 = UE.Wave Player:Mono(Wave Asset=@ArrayRandomGet_0.Value, Play=@n2.On Next, Pitch Shift=@n1.Value) {also {version [1,1]}}
   Multiply_0 = UE.Multiply:Audio by Float(AdditionalOperands=UE.Reroute(Value=@n2.Value).Value, PrimaryOperand=@WavePlayer_0.Out Mono)
  ]
  links [{dst="@UE.OutputFormat.Mono.Audio:0", src=Multiply_0.Out}, {dst="@UE.Source.OneShot.OnFinished", src=WavePlayer_0.On Finished}]
 }]
properties {
 assetImportData=&self:AssetImportData
 attenuationSettings=&/Att_Companion_Spark
 maxDistance=1200
 sampleRate=48000
 soundClassObject=&Fort_Creatures_Pets
 concurrencyOverrides {bLimitToOwner=true, maxCount=1, resolutionRule=PreventNew}
}
