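# Sound cue for the shared generic chewing sound that cosmetic companions make while eating.
# A Random node picks among twelve equally weighted chew waves and feeds a Modulator that varies pitch and volume on each play.
# The cue uses an eating attenuation asset and the creatures/pets sound class, with a boosted volume and a concurrency override that caps simultaneous chews.
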SoundCue /CosmeticCompanions/Assets/Audio_Shared/Generic/SC_Companion_Eat_Chew
root=Modulator_0
nodes [
 {id=Modulator_0, children [Random_0], properties {pitchMax=1.2, volumeMax=1}}
 {
  id=Random_0
  children [
   WavePlayer_0
   WavePlayer_1
   WavePlayer_2
   WavePlayer_3
   WavePlayer_4
   WavePlayer_5
   WavePlayer_6
   WavePlayer_7
   WavePlayer_8
   WavePlayer_9
   WavePlayer_10
   WavePlayer_11
  ]
  properties {weights [1,1,1,1,1,1,1,1,1,1,1,1]}
 }
 {id=WavePlayer_0, properties {soundWaveAssetPtr=/CosmeticCompanions/Assets/Audio_Shared/Generic/waves/Companion_Eat_Chew_12}}
 {id=WavePlayer_1, properties {soundWaveAssetPtr=/CosmeticCompanions/Assets/Audio_Shared/Generic/waves/Companion_Eat_Chew_11}}
 {id=WavePlayer_10, properties {soundWaveAssetPtr=/CosmeticCompanions/Assets/Audio_Shared/Generic/waves/Companion_Eat_Chew_02}}
 {id=WavePlayer_11, properties {soundWaveAssetPtr=/CosmeticCompanions/Assets/Audio_Shared/Generic/waves/Companion_Eat_Chew_01}}
 {id=WavePlayer_2, properties {soundWaveAssetPtr=/CosmeticCompanions/Assets/Audio_Shared/Generic/waves/Companion_Eat_Chew_10}}
 {id=WavePlayer_3, properties {soundWaveAssetPtr=/CosmeticCompanions/Assets/Audio_Shared/Generic/waves/Companion_Eat_Chew_09}}
 {id=WavePlayer_4, properties {soundWaveAssetPtr=/CosmeticCompanions/Assets/Audio_Shared/Generic/waves/Companion_Eat_Chew_08}}
 {id=WavePlayer_5, properties {soundWaveAssetPtr=/CosmeticCompanions/Assets/Audio_Shared/Generic/waves/Companion_Eat_Chew_07}}
 {id=WavePlayer_6, properties {soundWaveAssetPtr=/CosmeticCompanions/Assets/Audio_Shared/Generic/waves/Companion_Eat_Chew_06}}
 {id=WavePlayer_7, properties {soundWaveAssetPtr=/CosmeticCompanions/Assets/Audio_Shared/Generic/waves/Companion_Eat_Chew_05}}
 {id=WavePlayer_8, properties {soundWaveAssetPtr=/CosmeticCompanions/Assets/Audio_Shared/Generic/waves/Companion_Eat_Chew_04}}
 {id=WavePlayer_9, properties {soundWaveAssetPtr=/CosmeticCompanions/Assets/Audio_Shared/Generic/waves/Companion_Eat_Chew_03}}
]
properties {
 attenuationSettings=&/Att_Companion_Eat
 bOverrideConcurrency=true
 maxDistance=1200
 soundClassObject=&Fort_Creatures_Pets
 volumeMultiplier=2
 concurrencyOverrides {maxCount=3}
}
